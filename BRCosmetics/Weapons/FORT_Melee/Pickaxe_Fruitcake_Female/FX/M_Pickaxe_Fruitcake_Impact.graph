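# Unlit translucent particle material for the Fruitcake female pickaxe impact effect, usable on Niagara sprites and meshes.
# A noise normal map distorts the UVs of panning flame and smoke textures, shaped by a Fresnel whose exponent is driven by a dynamic parameter.
# The Switch_ scalar blends two depth-faded opacity variants built from those masks.
# A bright time-of-day color, desaturated by dynamic Param2, is multiplied by particle color for emissive adjustments.
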Material /BRCosmetics/Weapons/FORT_Melee/Pickaxe_Fruitcake_Female/FX/M_Pickaxe_Fruitcake_Impact
graph {
 properties {
  bUsedWithNiagaraMeshParticles=true
  bUsedWithNiagaraSprites=true
  blendMode=BLEND_Translucent
  pixelDepthOffsetMode=PDOM_Legacy
  shadingModel=MSM_Unlit
  usedShadingModels=Unlit
 }
 nodes [
  n1 = ParticleColor()
  n2 = TextureSample(UVs=Panner() {speedX=0.4, speedY=-0.2}) {samplerType=SAMPLERTYPE_Normal, texture=/Game/Characters/Player/Female/Medium/Bodies/F_MED_Bistro_Astronaut/FX/T_GaussianNoise_Normal}
  n3 = Add(A=TextureCoordinate() {uTiling=0.5, vTiling=0.5}, B=Multiply(A=ComponentMask(None=@n2.RGB) {g=true, r=true}) {constB=0.15})
  n4 = TextureSampleParameter2D(UVs=Panner(Coordinate=@n3) {speedX=0.6}) {
   parameterName=Texture
   samplerType=SAMPLERTYPE_LinearGrayscale
   texture=/BRCosmetics/Weapons/FORT_Melee/Pickaxe_Cube_Ninja_Male/FX/T_FluidFlameTile
  }
  n5 = Multiply(
   A=TextureSample(UVs=Panner() {speedX=1}) {samplerType=SAMPLERTYPE_Grayscale, texture=&LowResBlurredNoise}.R
   B=LinearInterpolate(Alpha=@n4.R) {constA=0.1}
  )
  n6 = Fresnel(ExponentIn=LinearInterpolate(Alpha=DynamicParameter() {paramNames [Fresnel,Param2,Param3,Param4]}.Fresnel) {constA=4, constB=1.44}) {baseReflectFraction=0, exponent=1}
  n7 = Multiply(A=@n5, B=@n6)
  n8 = Constant() {r=45}
  n9 = Panner(Coordinate=@n3) {speedX=0.55}
  n10 = Subtract(A=@n6, B=Multiply(A=TextureSample(UVs=@n9) {texture=&T_Inky_Smoke_Tile}.R) {constB=0.2})
  LinearInterpolate(
   A=Multiply(A=@n1.A, B=DepthFade(Opacity=@n7, FadeDistance=@n8))
   B=Multiply(A=DepthFade(Opacity=@n10, FadeDistance=@n8), B=ScalarParameter() {defaultValue=1, parameterName=Opacity})
   Alpha=ScalarParameter() {parameterName=Switch_}
  )
  n11 = VectorParameter() {parameterName=Color_Morning, defaultValue=<rgba 1 0 0.1729 1>}
  n12 = VectorParameter() {parameterName=Color_Day, defaultValue=<rgba 1 0 0.1729 1>}
  n13 = VectorParameter() {parameterName=Color_Evening, defaultValue=<rgba 1 0 0.1243 1>}
  n14 = VectorParameter() {parameterName=Color_Night, defaultValue=<rgba 1 0.0404 0 1>}
  n15 = Desaturation(
   None=Multiply(
    A=MaterialFunctionCall(Color_Morning=@n11.RGB, Color_Day=@n12.RGB, Color_Evening=@n13.RGB, Color_Night=@n14.RGB) {materialFunction=&Color_Time_of_Day}.Result
    B=ScalarParameter() {defaultValue=78}
   )
   Fraction=DynamicParameter() {paramNames [Fresnel,Param2,Param3,Param4]}.Param2
  )
  LinearInterpolate(A=@n15, B=Multiply(A=@n15, B=@n5), Alpha=ScalarParameter() {parameterName=Switch_})
  MaterialFunctionCall() {materialFunction=&DebugTimeSine}
  MaterialFunctionCall(Emissive=Multiply(A=@n15, B=@n1.RGB)) {materialFunction=&MF_FX_EmissiveAdjustments}
  ScalarParameter() {defaultValue=-0.5, parameterName=Desaturation}
 ]
}
